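# Physics asset for the Wise Master pickaxe, a two-handed melee weapon cosmetic with dangling rope and hair parts.
# A long capsule covers the handle, while boxes form the rope and hair bodies and their master anchors.
# Three rope segments chain from the rope master and one hair body hangs from the hair master, each joint with swing and twist limits.
# No constraint links the handle or either master to each other.
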PhysicsAsset /BRCosmetics/Weapons/FORT_Melee/Pickaxe_WiseMaster/Meshes/Wise_Master_Axe_PhysicsAsset
bodies {
 Melee_TwoHanded_Handle [{shapeType=Capsule, length=100.8892, radius=2.0521, rotation=<rot 0 -90 0>}]
 dyn_hair [{shapeType=Box, extentX=3.3428, extentY=4.0965, extentZ=9.1948, center=<4 0 0>, rotation=<rot 90 0 0>}]
 dyn_hair_mstr [{shapeType=Box, extentX=7.5, extentY=7.5, extentZ=7.5}]
 dyn_rope_01 [{shapeType=Box, extentX=2, extentY=2, extentZ=7, center=<3.25 0 0.0256>, rotation=<rot 90 0 0>}]
 dyn_rope_02 [{shapeType=Box, extentX=2, extentY=2, extentZ=7, center=<3.25 0 0.1398>, rotation=<rot 90 0 0>}]
 dyn_rope_03 [{shapeType=Box, extentX=2, extentY=2, extentZ=7, center=<3.25 0 -0.2103>, rotation=<rot 90 0 0>}]
 dyn_rope_mstr [{shapeType=Box, extentX=7.5, extentY=7.5, extentZ=7.5}]
}
constraints [
 {bone1Name=dyn_rope_01, bone2Name=dyn_rope_mstr, swing1LimitDegrees=70, swing2LimitDegrees=70, twistLimitDegrees=45}
 {bone1Name=dyn_rope_02, bone2Name=dyn_rope_01, swing1LimitDegrees=70, swing2LimitDegrees=70, twistLimitDegrees=45}
 {bone1Name=dyn_rope_03, bone2Name=dyn_rope_02, swing1LimitDegrees=70, swing2LimitDegrees=70, twistLimitDegrees=45}
 {bone1Name=dyn_hair, bone2Name=dyn_hair_mstr, swing1LimitDegrees=65, swing2LimitDegrees=65, twistLimitDegrees=45}
]
